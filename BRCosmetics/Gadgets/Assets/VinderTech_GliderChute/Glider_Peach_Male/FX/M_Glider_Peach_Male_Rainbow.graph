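# Translucent, unlit, two-sided material for the rainbow effect on the Peach Male glider, usable on Niagara ribbons and meshes.
# Color adds a gradient texture to smoke-distorted caustics, sampled at three offset pans for RGB fringing, tinted by particle color.
# Vertex offset math sums U-weighted sine waves and an optional clamped, Niagara-driven bend.
# Alpha combines particle alpha, depth and view-angle fades, edge and end masks, and a dynamic Rainbow_Mask wipe along U.
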Material /BRCosmetics/Gadgets/Assets/VinderTech_GliderChute/Glider_Peach_Male/FX/M_Glider_Peach_Male_Rainbow
graph {
 properties {
  bUsedWithMeshDeformer=true
  bUsedWithNiagaraMeshParticles=true
  bUsedWithNiagaraRibbons=true
  bUsedWithSkeletalMesh=true
  blendMode=BLEND_Translucent
  pixelDepthOffsetMode=PDOM_Legacy
  shadingModel=MSM_Unlit
  twoSided=true
  usedShadingModels=Unlit
 }
 nodes [
  n1 = DynamicParameter() {desc=Parameters set by Niagara, parameterIndex=1, paramNames [X,Y,Z,WPOIntensity]}
  n2 = Subtract(
   A=MaterialFunctionCall(X=@n1.X, Y=@n1.X, Z=@n1.X) {materialFunction=&MakeFloat3}.Result
   B=ParticlePositionWS()
  )
  n3 = Clamp(
   None=@n2
   Min=ScalarParameter() {defaultValue=-150, group=Interpolate WPO, parameterName=Clamp_min}
   Max=ScalarParameter() {defaultValue=150, group=Interpolate WPO, parameterName=Clamp_max}
  ) {maxDefault=1000, minDefault=-1000}
  n4 = TextureCoordinate()
  n5 = Multiply(
   A=Power(Base=ComponentMask(None=@n4) {r=true}, Exp=ScalarParameter() {defaultValue=3, group=Interpolate WPO, parameterName=WPOBendPower})
   B=@n1.WPOIntensity
  )
  n6 = Power(Base=ComponentMask(None=TextureCoordinate()) {r=true})
  n7 = Constant3Vector() {constant=<rgba 1 0 0 0>}
  n8 = TextureCoordinate()
  n9 = Time()
  n10 = Sine(None=Add(A=ComponentMask(None=@n8) {r=true}, B=Multiply(A=@n9) {constB=-1})) {period=1.2}
  n11 = Multiply(
   A=Transform(None=Multiply(A=@n7, B=@n10)) {transformSourceType=TRANSFORMSOURCE_Local}
   B=ScalarParameter() {defaultValue=2, group=WPO, parameterName=WPO_Internity_X}
  ) {constB=2}
  n12 = Power(Base=ComponentMask(None=TextureCoordinate()) {r=true})
  n13 = Constant3Vector() {constant=<rgba 0 0 1 0>}
  n14 = TextureCoordinate()
  n15 = Time()
  n16 = Sine(None=Add(A=ComponentMask(None=@n14) {r=true}, B=Multiply(A=@n15) {constB=-1}))
  n17 = Multiply(
   A=Transform(None=Multiply(A=@n13, B=@n16)) {transformSourceType=TRANSFORMSOURCE_Local}
   B=ScalarParameter() {defaultValue=5, group=WPO, parameterName=WPO_Internity_Z}
  ) {constB=5}
  Add(
   A=Add(A=Multiply(A=@n6, B=@n11), B=Multiply(A=@n12, B=@n17))
   B=LinearInterpolate(A=Constant(), B=LinearInterpolate(B=@n3, Alpha=@n5), Alpha=ScalarParameter() {group=Interpolate WPO, parameterName=Use interpolate WPO})
  )
  n18 = MaterialFunctionCall("FadeSensitivity (optional)"=ScalarParameter() {defaultValue=0.5, parameterName=FadeBasedOnViewAngle}) {materialFunction=&FadeBasedOnViewAngle}
  LinearInterpolate(B=OneMinus(None=@n18.Result), Alpha=ScalarParameter() {parameterName=Use_FadeBasedOnViewAngle})
  n19 = TextureCoordinate() {uTiling=0.3}
  n20 = Constant() {r=0.03}
  n21 = Panner(Coordinate=Add(A=@n19, B=@n20) {constB=0.05}) {speedX=-0.4}
  n22 = TextureCoordinate() {uTiling=0.7}
  n23 = Multiply(A=TextureSample(UVs=Panner(Coordinate=@n22) {speedX=-0.1}) {texture=&T_Inky_Smoke_Tile}.R) {constB=0.1}
  n24 = Panner(Coordinate=@n19) {speedX=-0.4}
  n25 = Panner(Coordinate=Add(A=@n19, B=Multiply(A=@n20) {constB=-1}) {constB=-0.05}) {speedX=-0.4}
  n26 = MaterialFunctionCall(
   X=TextureSample(UVs=Add(A=@n21, B=@n23)) {samplerType=SAMPLERTYPE_LinearGrayscale, texture=/Game/Global/Textures/Default/Caustics/T_Nosie_Caustics_Grayscale}.R
   Y=TextureSample(UVs=Add(A=@n24, B=@n23)) {samplerType=SAMPLERTYPE_LinearGrayscale, texture=&T_Nosie_Caustics_Grayscale}.R
   Z=TextureSample(UVs=Add(A=@n25, B=@n23)) {samplerType=SAMPLERTYPE_LinearGrayscale, texture=&T_Nosie_Caustics_Grayscale}.R
  ) {materialFunction=&MakeFloat3}
  n27 = Power(Base=Multiply(A=Power(Base=@n26.Result) {constExponent=1}) {constB=2}) {constExponent=1.2}
  n28 = Multiply(A=LinearInterpolate(A=Round(None=@n27), B=@n27) {constAlpha=0.9}) {constB=1.5}
  n29 = Panner(Coordinate=TextureCoordinate() {uTiling=0.8, vTiling=2}) {speedX=-0.1, speedY=0.3}
  n30 = Saturate(None=LinearInterpolate(Alpha=TextureSample(UVs=@n29) {texture=/Game/Effects/Fort_Effects/Textures/Power_Source/T_Power_Source_Sphere_Texture}.R) {constA=-0.2})
  n31 = Desaturation(None=Multiply(A=Multiply(A=@n28, B=@n30)) {constB=1.3}, Fraction=Constant() {r=-0.5})
  n32 = Constant()
  n33 = ScalarParameter() {group=Colorful_Texture, parameterName=Remove_Colorful_texure}
  n34 = MaterialFunctionCall(XY=TextureCoordinate()) {materialFunction=&Swizzle}
  n35 = ScalarParameter() {defaultValue=1, group=Emissive, parameterName=UV_Tiling}
  n36 = VectorParameter() {group=Emissive, parameterName=UV_offset, defaultValue=<rgba 0 0 0 1>}
  n37 = TextureSampleParameter2D(UVs=Add(A=Multiply(A=@n34.YX, B=@n35) {constB=2}, B=ComponentMask(None=@n36.RGB) {g=true, r=true})) {
   group=Emissive
   parameterName=Default_Tex
   texture=/Game/Gadgets/Assets/VinderTech_GliderChute/Glider_Peach_Male/FX/T_Glider_Peach_rainbow_gradient
  }
  n38 = Desaturation(None=@n37.RGB, Fraction=ScalarParameter() {defaultValue=-0.3, group=Emissive, parameterName=Desaturation})
  n39 = Round(None=@n38)
  n40 = Constant() {r=0.95}
  n41 = Multiply(
   A=ParticleColor().RGB
   B=Add(A=LinearInterpolate(A=@n31, B=@n32, Alpha=@n33), B=LinearInterpolate(A=@n39, B=@n38, Alpha=@n40) {constAlpha=0.7})
  )
  MaterialFunctionCall(Emissive=@n41) {materialFunction=&MF_FX_EmissiveAdjustments}
  n42 = MaterialFunctionCall(Float3=@n2) {materialFunction=&BreakOutFloat3Components}
  n43 = ScalarParameter() {defaultValue=-30, group=Interpolate WPO, parameterName=XYClamp_min}
  n44 = ScalarParameter() {defaultValue=30, group=Interpolate WPO, parameterName=XYClamp_max}
  MaterialFunctionCall(
   X=Clamp(None=@n42.R, Min=@n43, Max=@n44) {maxDefault=1000, minDefault=-1000}
   Y=Clamp(None=@n42.G, Min=@n43, Max=@n44) {maxDefault=1000, minDefault=-1000}
   Z=@n3
  ) {materialFunction=&MakeFloat3}
  n45 = ScalarParameter() {group=Alpha, parameterName=NearCameraFade}
  n46 = Multiply(
   A=ParticleColor().A
   B=DepthFade(
    Opacity=MaterialFunctionCall(Distance=@n45) {materialFunction=&Near_Camera_Fade}.Result
    FadeDistance=ScalarParameter() {defaultValue=35, group=Alpha, parameterName=FadeDistance}
   ) {fadeDistanceDefault=50}
  )
  n47 = ComponentMask(None=TextureCoordinate()) {g=true}
  n48 = Multiply(A=@n47, B=OneMinus(None=@n47))
  n49 = Multiply(
   A=Saturate(None=Multiply(A=@n46, B=@n18.Result))
   B=MaterialFunctionCall(In=Clamp(None=@n48), Contrast=ScalarParameter() {parameterName=MaskContrast}) {materialFunction=&CheapContrast}.Result
  )
  n50 = Add(A=TextureCoordinate(), B=AppendVector(A=ScalarParameter() {parameterName=Tail_Edge_hardness Add}, B=Constant()))
  n51 = Multiply(A=OneMinus(None=ComponentMask(None=@n50) {r=true}), B=ScalarParameter() {defaultValue=2.5, group=Alpha, parameterName=Tail_Edge_hardness}) {constB=2.5}
  n52 = ScalarParameter() {defaultValue=2, group=Alpha, parameterName=Tail_Mask_Power}
  n53 = TextureCoordinate()
  n54 = AppendVector(A=ScalarParameter() {parameterName=Lead_Edge_hardness Add}, B=Constant())
  n55 = Multiply(A=ComponentMask(None=Add(A=@n53, B=@n54)) {r=true}, B=ScalarParameter() {defaultValue=2, group=Alpha, parameterName=Lead_Edge_hardness}) {constB=2}
  n56 = ScalarParameter() {defaultValue=2, group=Alpha, parameterName=Lead_Mask_Power}
  n57 = Multiply(A=Saturate(None=Power(Base=@n51, Exp=@n52) {constExponent=4}), B=Saturate(None=Power(Base=@n55, Exp=@n56)))
  n58 = TextureCoordinate()
  n59 = Add(
   A=OneMinus(None=ComponentMask(None=@n58) {r=true})
   B=LinearInterpolate(Alpha=DynamicParameter() {defaultValue=<rgba 1.2 1 1 1>, paramNames [Width,Rainbow_Mask,Param3,Param4]}.Rainbow_Mask) {constA=-1, constB=0.5}
  )
  Multiply(
   A=Multiply(A=Multiply(A=@n49, B=@n57), B=ScalarParameter() {defaultValue=1, parameterName=Alpha_Main})
   B=Saturate(None=Multiply(A=@n59) {constB=3})
  )
  OneMinus(None=@n4)
  Reroute(None=@n41)
 ]
}
